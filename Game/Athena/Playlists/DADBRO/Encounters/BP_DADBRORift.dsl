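; BP_DADBRORift is a rift actor for the DADBRO Athena playlist encounter that orients itself when play begins.
; On the authority only, it finds the DadBro gameplay mutator by tag through the Athena game mode and sets its rotation to that mutator's desired DadBro rotation.
; The construction script makes the root component Movable, while the overlap and tick events are empty.
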(event EventBeginPlay
  (Networking|SwitchHasAuthority
    (:Authority
      (bind _asfort_game_mode_athena (Utilities|Casting|CastToFortGameModeAthena (Game|GetGameMode)))
      (bind _returnvalue (Class|FortGameModeAthena|GetGameplayMutator _asfort_game_mode_athena "(TagName=\"Mutator.DADBRO\")"))
      (bind _asfort_athena_mutator_dad_bro (Utilities|Casting|CastToFortAthenaMutator_DadBro _returnvalue))
      (Transformation|SetActorRotation (Class|FortAthenaMutatorDadBro|GetDesiredDadBroRotation _asfort_athena_mutator_dad_bro)))))

(event Collision|EventActorBeginOverlap
  (params (OtherActor :type Actor%20Object%20Reference)))

(event EventTick
  (params (DeltaSeconds :type Float%20%28single-precision%29)))

(fn ConstructionScript
  (Transformation|SetMobility (Variables|Transformation|GetRootComponent) "Movable"))
